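# Niagara system for the PowerBeam vehicle boost cosmetic, producing lightning and a trailing beam at the boost position plus a per-car offset.
# Lightning004 runs two spawn groups through masked inputs to vary lifetime, orientation, scale and color, rendered by several mesh renderers and a sprite renderer.
# Ribbons_Lightning drives two ribbon renderers with backward velocity, fading source-movement inheritance and separate curve-driven width and color.
# Both emitters apply painted-variant color overrides with R1 sets, and dim or tint differently when painted.
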NiagaraSystem /VehicleCosmetics/Boosts/Boost_PowerBeam/FX/NS_VC_Boost_PowerBeam
dependencies {
 usedDataInterfaces [
  &NiagaraDataInterfaceCurve
  &NiagaraDataInterfaceVectorCurve
  &NiagaraDataInterfaceMeshRendererInfo
  &NiagaraDataInterfaceSkeletalMesh
 ]
 usedDynamicInputs []
 usedModules [
  &self:SystemSpawnScript.SystemScriptSource.SystemScriptGraph.NiagaraNodeAssignment_0.SetVariables_#1
  &SystemState
 ]
}
emitters {
 Lightning004 {
  emitterUpdateScript [
   &EmitterState()
   &VC_FilteredSocketCount() {enabled=false}
   &SpawnRate(SpawnRate=20, Use Spawn Probability=false)
   &SpawnRate(SpawnGroup=1, SpawnRate=35, Use Spawn Probability=false) {module=SpawnRate001}
  ]
  particleSpawnScript [
   &VC_PaintedVariantsOverride(
    Black=<rgba 0.023 0.023 0.023 3>
    BurntSienna=<rgba 0.21 0.0329 0.0021 3>
    Cobalt=<rgba 0.045 0.0917 0.3 3>
    Crimson=<rgba 0.3 0.006 0.0109 3>
    DefaultColor=&MaskLinearColorBySpawnGroup(Masked Linear Color=<rgba 0 0.7083 0.85 1>, Passthrough Linear Color=<rgba 0.5 0.5 0.5 2>)
    ForestGreen=<rgba 0.1125 0.3 0.075 3>
    Grey=<rgba 0.1 0.1 0.1 3>
    Lime=<rgba 0.25 0.5 0 3>
    Orange=<rgba 0.35 0.0408 0 3>
    Pink=<rgba 0.45 0.0539 0.306 3>
    Purple=<rgba 0.1625 0.05 0.5 3>
    R1 Black:LinearColor=<rgba 0.04 0.04 0.04 1>
    R1 BurntSienna:LinearColor=<rgba 0.2 0.0423 0 1>
    R1 Cobalt:LinearColor=<rgba 0 0.1167 0.7 1>
    R1 DefaultColor=<rgba 0 0.8333 1 1>
    R1 ForestGreen:LinearColor=<rgba 0.1684 0.45 0.1125 1>
    R1 Grey:LinearColor=<rgba 0.15 0.15 0.15 1>
    R1 Lime:LinearColor=<rgba 0.275 0.55 0 1>
    R1 Orange:LinearColor=<rgba 0.4 0.0467 0 1>
    R1 Pink:LinearColor=<rgba 0.5052 0.0605 0.3435 1>
    R1 Purple:LinearColor=<rgba 0.1954 0.034 0.6795 1>
    R1 Saffron:LinearColor=<rgba 0.6 0.474 0.06 1>
    R1 TitaniumWhite:LinearColor=<rgba 0.6 0.6 0.6 1>
    Saffron=<rgba 0.3 0.237 0.03 1>
    TitaniumWhite=<rgba 0.4 0.4 0.4 2>
    Use R1 Painted Overrides=true
   )
   &InitializeParticle(
    Color=$StackContext.Color
    Lifetime=&MaskFloatBySpawnGroup(Masked Float=&RandomRangeFloat(Minimum=0.2, Maximum=0.35), Passthrough Float=0.2)
    Mesh Scale Max=<6 0.4 6>
    Mesh Scale Min=<6 7 10>
    Mesh Scale Mode=::NewEnumerator2
    Position Offset=&AddVectorToPosition(
     Position=&AddVectorToPosition(Position=<-10 0 0>, Vector=$User.SingleBoostPos)
     Vector=$User.DTCarOffset
    )
    UsePositionOffset=true
    Write Material Random=true
   )
   &PartitionParticles(Number Of Partitions=$Emitter.PortCount) {enabled=false}
   &SkeletalMeshLocation(
    Bone / Socket Index=$Particles.PartitionParticles.Partition
    Bone / Socket Sampling Mode=::NewEnumerator5
    MaskSpawnGroup=false
    Mesh Sampling Type=::NewEnumerator6
    Sampled Position Offset=<-185 0 24>
    Skeletal Mesh=$System.SkeletalMesh
   ) {enabled=false}
   &AddVelocity(Velocity=&MaskVectorBySpawnGroup(Masked Vector=&RandomRangeVector(Minimum=<-50 0 0>, Maximum=<-250 1 1>), Passthrough Vector=<0 0 0>)) {enabled=false}
   &ShapeLocation(
    MaskSpawnGroup=false
    Override Seed=false
    Sphere Radius=&MaskFloatBySpawnGroup(Masked Float=25, Passthrough Float=0)
    UseOwnerScale=true
   )
   &InitialMeshOrientation(
    Mesh Axis To Orient=&MaskVectorBySpawnGroup(
     Masked Vector=&RandomRangeVector(Minimum=<-1 -1 -0.7>, Maximum=<1 1 -1.8>)
     Passthrough Vector=<0.2 1 0>
    )
    Mesh Orientation Mode=::NewEnumerator2
    Rotation=&MaskVectorBySpawnGroup(
     Masked Vector=&RandomRangeVector(Minimum=<0.5 0.5 1>, Maximum=<-0.5 -0.5 1>)
     Passthrough Vector=&RandomRangeVector(Minimum=<0 -0.3 0>, Maximum=<0 0.3 0>)
    )
    Use Rotation Vector=true
   )
   /SetVariables(
    Particles.LightningVis:NiagaraInt32=&SpawnGroup()
    Particles.SizeLightningStrike:Vector3f=&RandomRangeVector(Minimum=<5 1 5>, Maximum=<3.5 5.6 3.5>)
   ) {isSetParameters=true}
  ]
  particleUpdateScript [
   &ParticleState()
   &ScaleMeshSize(Scale Factor=&MaskVectorBySpawnGroup(
     Masked Vector=<1 1 1>
     Passthrough Vector=&VectorFromCurve(VectorCurve=DataInterface() {propertyValues {XCurve {keys [<at 0 0.6>,<at 1 1>]}, YCurve {keys [<at 0 0.4>,<at 1 1>]}, ZCurve {keys [<at 0 0.6>,<at 1 1>]}}})
    ))
   /SetVariables(Particles.ScaleLightningStrike:Vector3f=&Multiply_Vector(
     A=$Particles.SizeLightningStrike
     B=&VectorFromCurve(VectorCurve=DataInterface() {propertyValues {XCurve {keys [<at 0 0.6>,<at 1 1>]}, YCurve {keys [<at 0 0.4>,<at 1 1>]}, ZCurve {keys [<at 0 0.6>,<at 1 1>]}}})
    )) {isSetParameters=true}
   &SolveForcesAndVelocity(Clamp Velocity=false, Limit Acceleration=false)
   &ScaleColor(
    Scale RGB=&MaskVectorBySpawnGroup(
     Masked Vector=&MakeCustomVectorFromBool(
      Bool=$User.bIsPainted
      True Vector=&VectorFromFloat(Value=0.7)
      False Vector=&VectorFromFloat(Value=2)
     )
     Passthrough Vector=&MakeCustomVectorFromBool(
      Bool=$User.bIsPainted
      True Vector=&VectorFromFloat(Value=0.65)
      False Vector=&VectorFromCurve(VectorCurve=DataInterface() {propertyValues {XCurve {keys [<at 0.5 0.3>,<at 1 0>]}, YCurve {keys [<at 0.5 0>,<at 1 1>]}, ZCurve {keys [<at 0 1>,<at 1 1>]}}})
     )
    )
    ScaleA=true
    ScaleRGB=true
   ) {module=ScaleColor001}
   &ScaleColor(
    Scale RGB=&MaskVectorBySpawnGroup(
     Masked Vector=&MakeCustomVectorFromBool(Bool=$User.bIsPainted, True Vector=&VectorFromFloat(Value=0.7), False Vector=<2 2 2>)
     Passthrough Vector=&MakeCustomVectorFromBool(
      Bool=$User.bIsPainted
      True Vector=&VectorFromFloat(Value=0.65)
      False Vector=&VectorFromCurve(VectorCurve=DataInterface() {propertyValues {XCurve {keys [<at 0.5 0.3>,<at 1 0>]}, YCurve {keys [<at 0.5 0>,<at 1 1>]}, ZCurve {keys [<at 0 1>,<at 1 1>]}}})
     )
    )
    ScaleA=true
    ScaleRGB=true
   ) {enabled=false}
   &SubUVAnimation(
    Mesh Renderer=DataInterface() {propertyValues {MeshRenderer=&self:Lightning004_2.NiagaraMeshRendererProperties_0}}
    Renderer Type:ENiagaraMeshOrSprite=::NewEnumerator1
    SubUV Animation Mode:ENiagara_SubUVLookupModeV2=::NewEnumerator1
    SubUV Curve=DataInterface() {propertyValues {Curve {keys [{tangentMode=RCTM_User, time=0, value=0}, {tangentMode=RCTM_User, time=1, value=1}]}}}
    UseEndFrame=false
    UseStartFrame=false
   )
   &DynamicMaterialParameters(
    Index 0 Param 2=2
    Index 0 Param 3=&FloatFromCurve(FloatCurve=DataInterface() {propertyValues {Curve {defaultValue=1, keys [{tangentMode=RCTM_User, time=0, value=0}, {tangentMode=RCTM_User, time=1, value=3}]}}})
    Module.Index0Float0Write=true
    Module.Index0Float1Write=true
    Module.Index0Float2Write=true
    Module.Index0Float3Write=true
   )
  ]
  rendererClasses [&NiagaraMeshRendererProperties, &NiagaraSpriteRendererProperties]
  renderers [
   {rendererClass=&NiagaraMeshRendererProperties, rendererIndex=0}
   {rendererClass=&NiagaraMeshRendererProperties, rendererIndex=1}
   {rendererClass=&NiagaraMeshRendererProperties, rendererIndex=2}
   {rendererClass=&NiagaraMeshRendererProperties, rendererIndex=3}
   {rendererClass=&NiagaraMeshRendererProperties, rendererIndex=4}
   {rendererClass=&NiagaraSpriteRendererProperties, rendererIndex=5}
  ]
 }
 Ribbons_Lightning {
  emitterSpawnScript [/SetVariables(Emitter.R1 width=170) {isSetParameters=true}]
  emitterUpdateScript [
   &EmitterState()
   &VC_FilteredSocketCount()
   &VC_RateQualityHandler()
   &VC_SpawnRibbon_v1(SpawnRate=$Emitter.Spawn Rate)
  ]
  particleSpawnScript [
   &VC_PaintedVariantsOverride(
    Black=<rgba 0.14 0.14 0.14 2>
    BurntSienna=<rgba 0.39 0.1635 0.0585 3>
    Cobalt=<rgba 0.12 0.24 0.6 1.2>
    Crimson=<rgba 0.6 0.045 0.082 2>
    DefaultColor=<rgba 0.7 0 0.525 2>
    ForestGreen=<rgba 0.3167 0.5 0.225 2>
    Grey=<rgba 0.27 0.27 0.27 1.5>
    Lime=<rgba 0.4372 0.55 0.099 2>
    Orange=<rgba 0.6 0.1884 0.012 2>
    Pink=<rgba 0.6 0.12 0.52 2>
    Purple=<rgba 0.341 0.18 0.6 2>
    R1 Black:LinearColor=<rgba 0.06 0.06 0.06 1>
    R1 BurntSienna:LinearColor=<rgba 0.6 0.1459 0.024 1>
    R1 Cobalt:LinearColor=<rgba 0.03 0.0849 3 1>
    R1 Crimson:LinearColor=<rgba 1 0.02 0.0363 1>
    R1 DefaultColor=<rgba 0 3 3 1>
    R1 ForestGreen:LinearColor=<rgba 0 1.5 0.25 1>
    R1 Lime:LinearColor=<rgba 1.4167 2 0.6 1>
    R1 Orange:LinearColor=<rgba 0.8 0.279 0 1>
    R1 Pink:LinearColor=<rgba 2 0.2396 1.3598 1>
    R1 Purple:LinearColor=<rgba 1.2312 0 3 1>
    R1 Saffron:LinearColor=<rgba 1 0.8017 0.15 1>
    R1 SkyBlue:LinearColor=<rgba 0.02 0.6721 1 1>
    R1 TitaniumWhite:LinearColor=<rgba 0 0 0 1>
    Saffron=<rgba 0.65 0.5287 0.13 2>
    SkyBlue=<rgba 0 0.5 0.6 2>
    TitaniumWhite=<rgba 0.58 0.58 0.58 2>
    Use R1 Painted Overrides=true
   )
   &VC_InitializeRibbon_v1(
    Color=$StackContext.Color
    Custom Offset=true
    Lifetime=1
    Local Vector Offset=&AddVectorToPosition(
     Position=&AddVectorToPosition(Position=<-10 0 0>, Vector=$User.SingleBoostPos)
     Vector=$User.DTCarOffset
    )
    Offset Amount=1
    Ribbon Width=130
   )
   &AddVelocity(Velocity=<-1400 0 0>) {module=AddVelocity001}
   /VehicleCosmetics/FX/Niagara/Modules/MergeEmitters/VC_ExtraBindingColors(
    ExtraColors=1
    R1 Color:LinearColor=&MakeCustomLinearColorFromBool(Bool=$User.bIsPainted, True Color=$User.PaintedColor, False Color=<rgba 0 2 2 0>)
   ) {enabled=false}
  ]
  particleUpdateScript [
   /SetVariables(Particles.R1 Color:LinearColor=&MakeCustomLinearColorFromBool(Bool=$User.bIsPainted, True Color=$User.PaintedColor, False Color=<rgba 0 2 2 1>)) {enabled=false, isSetParameters=true}
   &ParticleState()
   &InheritSourceMovement(
    Applied Position Scale=&Multiply_VectorByFloat(
     Vector=<1 1 1>
     Float=&FloatFromCurve(FloatCurve=DataInterface() {propertyValues {Curve {keys [{interpMode=RCIM_Cubic, tangentMode=RCTM_Break, time=0, value=1}, {interpMode=RCIM_Cubic, tangentMode=RCTM_Break, time=1, value=0}]}}})
    )
    ApplyForce=false
    ApplyPosition=true
    ApplyPositionFalloff=false
   )
   /SetVariables(
    Particles.R1BlueLightningColor:LinearColor=&self:VC_Dynamic_ScaleColor_01(
     Alpha Scale:NiagaraFloat=&FloatFromCurve(FloatCurve=DataInterface() {propertyValues {Curve {keys [<at 0.03 0>,<at 0.2 1>,<at 0.8 1>,<at 1 0>]}}})
     Color to Scale:LinearColor=$Particles.R1 Color
     RGB Scale:Vector3f=&VectorFromFloat(Value=4.5)
    )
    Particles.R1VBlueLightning:Vector3f=<-1650 0 0>
    Particles.R1WBlueLightning:NiagaraFloat=&Multiply_Float(
     A=$Emitter.R1 width
     B=&FloatFromCurve(FloatCurve=DataInterface() {propertyValues {Curve {keys [<at 0.01 0>,<at 0.05 0.425>,<at 0.0715 0.6625>,<at 0.1378 0.9>,<at 0.35 1.2>,<at 0.6 0.8>,<at 0.73 0>]}}})
    )
   ) {isSetParameters=true}
   &SolveForcesAndVelocity(Clamp Velocity=false, Limit Acceleration=false)
   &VC_UpdateRibbon_v1()
   &DynamicMaterialParameters(
    Index 0 Param 2=2
    Index 0 Param 3=1
    Index 1 Param 1=3
    Index 1 Param 2=1.25
    Module.Index0Param0Write=true
    Module.Index0Param1Write=true
    Module.Index0Param2Write=true
    Module.Index0Param3Write=true
    Module.Index1Param0Write=true
    Module.Index1Param1Write=true
    Module.Index1Param2Write=true
    Module.Index1Param3Write=true
    Write Parameter Index 1=true
   )
   &ScaleColor(
    Scale Alpha=&FloatFromCurve(FloatCurve=DataInterface() {propertyValues {Curve {keys [
         {arriveTangent=-1.661, interpMode=RCIM_Cubic, leaveTangent=-1.661, tangentMode=RCTM_User, time=0, value=1}
         {arriveTangent=-1.8893, interpMode=RCIM_Cubic, leaveTangent=-1.8893, tangentMode=RCTM_User, time=0.8, value=0.5}
         <at 1 0>
        ]}}})
    Scale RGB=&MakeCustomVectorFromBool(
     Bool=$User.bIsPainted
     True Vector=&Multiply_VectorByFloat(
      Vector=<1 1 1>
      Float=&FloatFromCurve(FloatCurve=DataInterface() {propertyValues {Curve {keys [<at 0 1 RCIM_Cubic>,<at 0.7 0.2 RCIM_Cubic>]}}})
     )
     False Vector=&VectorFromCurve(VectorCurve=DataInterface() {propertyValues {XCurve {keys [<at 0 1>,<at 0.3 0.2>]}, YCurve {keys [<at 0 0>,<at 1 0>]}, ZCurve {keys [<at 0 2>,<at 0.3 1>]}}})
    )
    ScaleA=false
    ScaleRGB=true
   )
   &ScaleRibbonWidth(Ribbon Width Scale=&FloatFromCurve(FloatCurve=DataInterface() {propertyValues {Curve {keys [
         {interpMode=RCIM_Cubic, tangentMode=RCTM_User, time=0, value=0}
         {
          arriveTangentWeight=0.1403
          interpMode=RCIM_Cubic
          leaveTangentWeight=0.1403
          tangentMode=RCTM_User
          tangentWeightMode=RCTWM_WeightedBoth
          time=0.27
          value=1
         }
         <at 0.8 0.3>
         <at 1 0>
        ]}}}))
  ]
  rendererClasses [&NiagaraRibbonRendererProperties]
  renderers [
   {rendererClass=&NiagaraRibbonRendererProperties}
   {rendererClass=&NiagaraRibbonRendererProperties, rendererIndex=1}
  ]
 }
}
summary {
 systemName=NS_VC_Boost_PowerBeam
 emitters [
  {emitterName=Ribbons_Lightning, rendererClasses [&NiagaraRibbonRendererProperties]}
  {emitterName=Lightning004, rendererClasses [&NiagaraMeshRendererProperties, &NiagaraSpriteRendererProperties]}
 ]
 userVariables [
  User.bIsFrontendPreview {type=NiagaraBool}
  User.bIsPainted {type=NiagaraBool}
  User.ColorVariantID {type=NiagaraInt32}
  User.DTCarOffset {type=Vector3f, description="", defaultValue {struct=Vector3f, value=<0 0 0>}}
  User.EmissiveLightning {type=NiagaraFloat, defaultValue=2, description=""}
  User.PaintedColor {type=LinearColor, description="", defaultValue {struct=LinearColor, value=<rgba 1 1 1 1>}}
  User.PaintedScalar {type=NiagaraFloat, defaultValue=1, description=""}
  User.PortCount {type=NiagaraInt32, defaultValue=1}
  User.SingleBoostPos {type=Vector3f, description="", defaultValue {struct=Vector3f, value=<0 0 0>}}
 ]
}
